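# Conversation database for the TideKing NPC, defining the greeting and top-level branches when a player talks to him.
# One entry tag leads to a Speech node whose default line is overridden by contextual messages for first-ever meetings, specific player outfits and per-match conversation count.
# Curve rows set priority for first-meeting and outfit lines.
# The speech fans out through link nodes to shared quest and basic-service conversations, TideKing dialogue and a Season 29 Snapshot quest entry.
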ConversationDatabase /NPCLibrary/NPCs/TideKing/NPCLibrary_Conversation_TideKing
entry_points [{tag=Conversation.Entry.TideKing, destinations [#1]}]
nodes [
 ConversationEntryPointNode_7 {
  id=#1
  class=ConversationEntryPointNode
  outputs [#2]
  properties {entryTag {tagName=Conversation.Entry.TideKing}}
 }
 Speech_C_4 {
  id=#2
  class=Speech_C
  outputs [#3, #4, #5, #6]
  properties {
   nodeName=""
   generalConfig {
    defaultMessage="Mortal! Lookin' beachy-keen."
    contextualMessages [
     {
      message="Wassup mortal? Looks like you're ready to party!"
      contextRequirements [{requirement {class=FortControllerRequirement_IsFirstEverConversationWithNPC}}]
      priority {curve {rowName=Tandem.Speech.Priority.FirstEverInteraction}}
     }
     {
      message=Whoa. You glow like sunken treasure. What ARE you?
      contextRequirements [{requirement {
         class=FortControllerRequirement_HasCID
         properties {allowedCIds [&Character_VoidRedemption_Rebel]}
        }}]
      priority {curve {rowName=Tandem.Speech.Priority.OutfitReaction}}
     }
     {
      message="Zeus, hey! Listen, bro, could you tone down the lightning a bit? You're scaring the fish."
      contextRequirements [{requirement {
         class=FortControllerRequirement_HasCID
         properties {allowedCIds [&Character_StunningMountain_Armor]}
        }}]
      priority {curve {rowName=Tandem.Speech.Priority.OutfitReaction}}
     }
     {
      message="The best hound in the world! Ready to go for a swim, Cerber-buddy?"
      contextRequirements [{requirement {
         class=FortControllerRequirement_HasCID
         properties {allowedCIds [/BRCosmetics/Athena/Items/Cosmetics/Characters/Character_GateHound_Spectral]}
        }}]
      priority {curve {rowName=Tandem.Speech.Priority.OutfitReaction}}
     }
     {
      message="Artemis, hey! Catch anything good lately? This place has some awesome fishing spots."
      contextRequirements [{requirement {
         class=FortControllerRequirement_HasCID
         properties {allowedCIds [/BRCosmetics/Athena/Items/Cosmetics/Characters/Character_SeleneCobra_Stride]}
        }}]
      priority {curve {rowName=Tandem.Speech.Priority.OutfitReaction}}
     }
     {
      message="Hey, uh, Medusa. I'm glad you got the invite..."
      contextRequirements [{requirement {
         class=FortControllerRequirement_HasCID
         properties {allowedCIds [/BRCosmetics/Athena/Items/Cosmetics/Characters/Character_ReptilianOcean_Sleek]}
        }}]
      priority {curve {rowName=Tandem.Speech.Priority.OutfitReaction}}
     }
     {
      message="Mortal! Lookin' beachy-keen."
      contextRequirements [{requirement {
         class=FortControllerRequirement_CheckNPCMatchConversationCount
         properties {nPCUniqueIdTag {tagName=AISpawnerData.Type.Specialist.Heavy.TideKing}}
        }}]
      priority {curve {rowName=None}}
     }
     {
      message="Hey, I ever tell you about the dolphin surfing championship?"
      contextRequirements [{requirement {
         class=FortControllerRequirement_CheckNPCMatchConversationCount
         properties {nPCUniqueIdTag {tagName=AISpawnerData.Type.Specialist.Heavy.TideKing}}
        }}]
      priority {curve {rowName=None}}
     }
     {
      message="If you see Zeus, could you tell him to tone it down?"
      contextRequirements [{requirement {
         class=FortControllerRequirement_CheckNPCMatchConversationCount
         properties {nPCUniqueIdTag {tagName=AISpawnerData.Type.Specialist.Heavy.TideKing}}
        }}]
      priority {curve {rowName=None}}
     }
    ]
   }
  }
 }
 ConversationLinkNode_2 {
  id=#6
  class=ConversationLinkNode
  properties {remoteEntryTag {tagName=Conversation.Entry.Season29.Snapshot.P02.Q04}}
 }
 ConversationLinkNode_3 {id=#5, class=ConversationLinkNode, properties {remoteEntryTag {tagName=Conversation.Entry.TideKing.Dialogue}}}
 ConversationLinkNode_0 {id=#4, class=ConversationLinkNode, properties {remoteEntryTag {tagName=Conversation.Entry.BasicServices}}}
 ConversationLinkNode_1 {id=#3, class=ConversationLinkNode, properties {remoteEntryTag {tagName=Conversation.Entry.NPCLibrary.Quests}}}
]
