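# Conversation database for the Tide King NPC's dialogue, where the sea god, speaking as Poseidon, tells classic beach stories.
# A single entry point leads to an "Ask about..." speech that offers two story choices, Dolphin Surfing and Underwater Beach Volleyball, plus a Back option.
# Each story is a linear chain of speech nodes, and both converge on one shared return-to-start node labeled BACK.
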ConversationDatabase /NPCLibrary/NPCs/TideKing/NPCLibrary_Conversation_TideKing_Dialogue
entry_points [{tag=Conversation.Entry.TideKing.Dialogue, destinations [#1]}]
nodes [
 ConversationEntryPointNode_0 {
  id=#1
  class=ConversationEntryPointNode
  outputs [#2]
  properties {entryTag {tagName=Conversation.Entry.TideKing.Dialogue}}
 }
 Speech_C_0 {
  id=#2
  class=Speech_C
  outputs [#3, #4, #5]
  properties {
   nodeName=""
   softTaskIcon=&T_UI_NPC_question
   generalConfig {defaultMessage="You wanna hear a classic Poseidon beach story, mortal?"}
  }
  sub_nodes [ConversationChoiceNode_5 {class=ConversationChoiceNode, properties {defaultChoiceDisplayText=Ask about...}}]
 }
 Speech_C_4 {
  id=#5
  class=Speech_C
  outputs [#6]
  properties {
   nodeName=""
   softTaskIcon=&T_UI_NPC_question
   generalConfig {defaultMessage="Talk about an epic tournament. Even Zeus reared his crackly head!"}
  }
  sub_nodes [ConversationChoiceNode_3 {class=ConversationChoiceNode, properties {defaultChoiceDisplayText=Underwater Beach Volleyball}}]
 }
 Speech_C_5 {
  id=#6
  class=Speech_C
  outputs [#7]
  properties {
   nodeName=""
   softTaskIcon=&T_UI_NPC_question
   generalConfig {defaultMessage="Cerberus tosses me a serve, and I'm about to spike it down for the crown..."}
  }
 }
 Speech_C_6 {
  id=#7
  class=Speech_C
  outputs [#8]
  properties {
   nodeName=""
   softTaskIcon=&T_UI_NPC_question
   generalConfig {defaultMessage="...when an arrow knocks the ball out of the way! I dove after it, but the waves took it to the ground."}
  }
 }
 Speech_C_10 {
  id=#8
  class=Speech_C
  outputs [#9]
  properties {
   nodeName=""
   softTaskIcon=&T_UI_NPC_question
   generalConfig {defaultMessage="Next thing I know, Zeus and Ares are hoisting Artemis on their shoulders and laughing."}
  }
 }
 Speech_C_11 {
  id=#9
  class=Speech_C
  outputs [#10]
  properties {
   nodeName=""
   softTaskIcon=&T_UI_NPC_question
   generalConfig {defaultMessage=I didn't even get to do my cool victory pose.}
  }
 }
 ReturnToStart_C_1 {
  id=#10
  class=ReturnToStart_C
  properties {softTaskIcon=&T_UI_NPC_Back}
  sub_nodes [ConversationChoiceNode_1 {class=ConversationChoiceNode, properties {defaultChoiceDisplayText=BACK}}]
 }
 Back_C_3 {
  id=#4
  class=Back_C
  sub_nodes [ConversationChoiceNode_23 {class=ConversationChoiceNode, properties {defaultChoiceDisplayText=BACK}}]
 }
 Speech_C_3 {
  id=#3
  class=Speech_C
  outputs [#11]
  properties {
   nodeName=""
   softTaskIcon=&T_UI_NPC_question
   generalConfig {defaultMessage="It's like this right? Phinster, my pet dolphin, entered himself into a competition."}
  }
  sub_nodes [ConversationChoiceNode_0 {class=ConversationChoiceNode, properties {defaultChoiceDisplayText=Dolphin Surfing}}]
 }
 Speech_C_1 {
  id=#11
  class=Speech_C
  outputs [#12]
  properties {
   nodeName=""
   softTaskIcon=&T_UI_NPC_question
   generalConfig {defaultMessage="I didn't know about it, of course. Not till he dragged me into it."}
  }
 }
 Speech_C_2 {
  id=#12
  class=Speech_C
  outputs [#13]
  properties {
   nodeName=""
   softTaskIcon=&T_UI_NPC_question
   generalConfig {defaultMessage="Anyways, we're there, at the starting line, and guess who we see next to us? Go on, guess!"}
  }
 }
 Speech_C_7 {
  id=#13
  class=Speech_C
  outputs [#14]
  properties {nodeName="", softTaskIcon=&T_UI_NPC_question, generalConfig {defaultMessage="HADES! SURFIN' CERBERUS!"}}
 }
 Speech_C_8 {
  id=#14
  class=Speech_C
  outputs [#15]
  properties {
   nodeName=""
   softTaskIcon=&T_UI_NPC_question
   generalConfig {defaultMessage="Phinster and I won, of course. God of the sea, here."}
  }
 }
 Speech_C_9 {
  id=#15
  class=Speech_C
  outputs [#10]
  properties {
   nodeName=""
   softTaskIcon=&T_UI_NPC_question
   generalConfig {defaultMessage="But let me tell you, that Cerberus can SWIM!"}
  }
 }
]
